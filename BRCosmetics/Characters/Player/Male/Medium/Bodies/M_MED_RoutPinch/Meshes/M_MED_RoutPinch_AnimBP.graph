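# Post-process style AnimBP for the M_MED_RoutPinch male body cosmetic, adding secondary motion to cloth, wool and chain bones.
# It copies the parent mesh pose into a Control Rig fed clothing and slide flags, then runs serial spring-based AnimDynamics, some repeated per bone, with upward gravity on left cloth.
# The update event stores gravity, rigid-body and skydiving values from a borrowed F_MED_IndoorLaceScore parameter asset, but the AnimGraph reads none.
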AnimBlueprint /BRCosmetics/Characters/Player/Male/Medium/Bodies/M_MED_RoutPinch/Meshes/M_MED_RoutPinch_AnimBP
target_skeleton=&/M_MED_RoutPinch_Skeleton
graphs [
 AnimGraph {nodes [
   n1 = Control Rig!(
    Source=Copy Pose From Mesh() {bUseAttachedParent=true}
    Alpha
    clothing:bool=Get clothing() -> (clothing:bool)
    bIsSliding:bool=Get bIsSliding()
    bIsBaseballSliding:bool=Get bIsBaseballSliding()
   ) -> (Pose) {
    alphaBoolBlend {blendOption=Linear}
    controlRigAssetReference {blueprintRigClass=&/M_MED_RoutPinch_CtrlRig.M_MED_RoutPinch_CtrlRig_C}
   }
   n2 = Anim Dynamics - Bone: dyn_Cloth_volume_l_B01!(ComponentPose=Local To Component(LocalPose=@n1), Alpha=1, GravityOverride:Vector="0,0,500", ExternalForce="0,0,0") -> (Pose) {
    angularDampingOverride=0.9999
    angularSpringConstant=5000
    bAngularSpring=true
    bOverrideAngularDamping=true
    bUseGravityOverride=true
    gravityScale=0.3
    simulationSpace=BoneRelative
    alphaBoolBlend {blendOption=Linear}
    boundBone {boneName=dyn_Cloth_volume_l_B01}
    componentAppliedLinearAccClamp=<2500 2500 1500>
    componentLinearAccScale=<0.3 0.3 0.3>
    componentLinearVelScale=<0.35 0.35 0.35>
    gravityOverride {z=500}
    physicsBodyDefinitions [{
      collisionType=CustomSphere
      sphereCollisionRadius=2
      boundBone {boneName=dyn_Cloth_volume_l_B01}
      boxExtents=<45 45 45>
      constraintSetup {angularLimitsMax=<0 5 15>, angularLimitsMin=<0 -5 0>, angularTarget=<0 0 0>}
      localJointOffset=<-10 0.2 -0.1>
     }]
    relativeSpaceBone {boneName=root}
    simSpaceSettings {maxAngularAcceleration=500, maxAngularVelocity=50}
   }
   n3 = Anim Dynamics - Bone: dyn_Cloth_volume_l_B01(ComponentPose=@n2, Alpha=1, GravityOverride="0,0,500", ExternalForce="0,0,0") {
    angularDampingOverride=0.9999
    angularSpringConstant=5000
    bAngularSpring=true
    bOverrideAngularDamping=true
    bUseGravityOverride=true
    gravityScale=0.3
    simulationSpace=BoneRelative
    alphaBoolBlend {blendOption=Linear}
    boundBone {boneName=dyn_Cloth_volume_l_B01}
    componentAppliedLinearAccClamp=<2500 2500 1500>
    componentLinearAccScale=<0.3 0.3 0.3>
    componentLinearVelScale=<0.35 0.35 0.35>
    gravityOverride {z=500}
    physicsBodyDefinitions [{
      collisionType=CustomSphere
      sphereCollisionRadius=2
      boundBone {boneName=dyn_Cloth_volume_l_B01}
      boxExtents=<25 15 15>
      constraintSetup {angularLimitsMax=<0 5 15>, angularLimitsMin=<0 -5 0>, angularTarget=<0 0 0>}
      localJointOffset=<-10 0.2 -0.1>
     }]
    relativeSpaceBone {boneName=root}
    simSpaceSettings {maxAngularAcceleration=500, maxAngularVelocity=50}
   }
   n4 = Anim Dynamics - Bone: dyn_Cloth_volume_l_B02!(ComponentPose=@n3, Alpha=1, GravityOverride:Vector="0,0,500", ExternalForce="0,0,0") -> (Pose) {
    angularDampingOverride=0.9999
    angularSpringConstant=5000
    bAngularSpring=true
    bOverrideAngularDamping=true
    bUseGravityOverride=true
    gravityScale=0.3
    simulationSpace=BoneRelative
    alphaBoolBlend {blendOption=Linear}
    boundBone {boneName=dyn_Cloth_volume_l_B02}
    componentAppliedLinearAccClamp=<2500 2500 1500>
    componentLinearAccScale=<0.3 0.3 0.3>
    componentLinearVelScale=<0.35 0.35 0.35>
    gravityOverride {z=500}
    physicsBodyDefinitions [{
      collisionType=CustomSphere
      sphereCollisionRadius=2
      boundBone {boneName=dyn_Cloth_volume_l_B02}
      boxExtents=<18 10 10>
      constraintSetup {angularLimitsMax=<0 5 15>, angularLimitsMin=<0 -5 0>, angularTarget=<0 0 0>}
      localJointOffset=<-10 0.2 -0.1>
     }]
    relativeSpaceBone {boneName=root}
    simSpaceSettings {maxAngularAcceleration=500, maxAngularVelocity=50}
   }
   n5 = Anim Dynamics - Bone: dyn_Cloth_volume_r_B01!(ComponentPose=@n4, Alpha=1, GravityOverride:Vector="0,0,-980", ExternalForce="0,0,0") -> (Pose) {
    angularDampingOverride=0.9999
    angularSpringConstant=5000
    bAngularSpring=true
    bOverrideAngularDamping=true
    bUseGravityOverride=true
    gravityScale=0.3
    simulationSpace=BoneRelative
    alphaBoolBlend {blendOption=Linear}
    boundBone {boneName=dyn_Cloth_volume_r_B01}
    componentAppliedLinearAccClamp=<2500 2500 1500>
    componentLinearAccScale=<0.3 0.3 0.3>
    componentLinearVelScale=<0.35 0.35 0.35>
    gravityOverride {z=-980}
    physicsBodyDefinitions [{
      collisionType=CustomSphere
      sphereCollisionRadius=2
      boundBone {boneName=dyn_Cloth_volume_r_B01}
      boxExtents=<45 45 45>
      constraintSetup {angularLimitsMax=<0 5 15>, angularLimitsMin=<0 -5 0>, angularTarget=<0 0 0>}
      localJointOffset=<-10 0.2 -0.1>
     }]
    relativeSpaceBone {boneName=root}
    simSpaceSettings {maxAngularAcceleration=500, maxAngularVelocity=50}
   }
   n6 = Anim Dynamics - Bone: dyn_Cloth_volume_r_B01(ComponentPose=@n5, Alpha=1, GravityOverride="0,0,-980", ExternalForce="0,0,0") {
    angularDampingOverride=0.9999
    angularSpringConstant=5000
    bAngularSpring=true
    bOverrideAngularDamping=true
    bUseGravityOverride=true
    gravityScale=0.3
    simulationSpace=BoneRelative
    alphaBoolBlend {blendOption=Linear}
    boundBone {boneName=dyn_Cloth_volume_r_B01}
    componentAppliedLinearAccClamp=<2500 2500 1500>
    componentLinearAccScale=<0.3 0.3 0.3>
    componentLinearVelScale=<0.35 0.35 0.35>
    gravityOverride {z=-980}
    physicsBodyDefinitions [{
      collisionType=CustomSphere
      sphereCollisionRadius=2
      boundBone {boneName=dyn_Cloth_volume_r_B01}
      boxExtents=<25 15 15>
      constraintSetup {angularLimitsMax=<0 5 15>, angularLimitsMin=<0 -5 0>, angularTarget=<0 0 0>}
      localJointOffset=<-10 0.2 -0.1>
     }]
    relativeSpaceBone {boneName=root}
    simSpaceSettings {maxAngularAcceleration=500, maxAngularVelocity=50}
   }
   n7 = Anim Dynamics - Bone: dyn_Cloth_volume_r_B02!(ComponentPose=@n6, Alpha=1, GravityOverride:Vector="0,0,-980", ExternalForce="0,0,0") -> (Pose) {
    angularDampingOverride=0.9999
    angularSpringConstant=5000
    bAngularSpring=true
    bOverrideAngularDamping=true
    bUseGravityOverride=true
    gravityScale=0.3
    simulationSpace=BoneRelative
    alphaBoolBlend {blendOption=Linear}
    boundBone {boneName=dyn_Cloth_volume_r_B02}
    componentAppliedLinearAccClamp=<2500 2500 1500>
    componentLinearAccScale=<0.3 0.3 0.3>
    componentLinearVelScale=<0.35 0.35 0.35>
    gravityOverride {z=-980}
    physicsBodyDefinitions [{
      collisionType=CustomSphere
      sphereCollisionRadius=2
      boundBone {boneName=dyn_Cloth_volume_r_B02}
      boxExtents=<18 10 10>
      constraintSetup {angularLimitsMax=<0 5 15>, angularLimitsMin=<0 -5 0>, angularTarget=<0 0 0>}
      localJointOffset=<-10 0.2 -0.1>
     }]
    relativeSpaceBone {boneName=root}
    simSpaceSettings {maxAngularAcceleration=500, maxAngularVelocity=50}
   }
   n8 = Anim Dynamics - Bone: dyn_Cloth_volume_l_C01!(ComponentPose=@n7, Alpha=1, GravityOverride:Vector="0,0,500", ExternalForce="0,0,0") -> (Pose) {
    angularBiasOverride=1
    angularDampingOverride=0.9999
    angularSpringConstant=5000
    bAngularSpring=true
    bOverrideAngularBias=true
    bOverrideAngularDamping=true
    bUseGravityOverride=true
    gravityScale=0.3
    simulationSpace=BoneRelative
    alphaBoolBlend {blendOption=Linear}
    boundBone {boneName=dyn_Cloth_volume_l_C01}
    componentAppliedLinearAccClamp=<3000 3000 3000>
    componentLinearAccScale=<0.5 0.5 0.5>
    componentLinearVelScale=<0.55 0.55 0.55>
    gravityOverride {z=500}
    physicsBodyDefinitions [{
      collisionType=CustomSphere
      sphereCollisionRadius=2
      boundBone {boneName=dyn_Cloth_volume_l_C01}
      boxExtents=<45 45 45>
      constraintSetup {angularLimitsMax=<5 5 15>, angularLimitsMin=<-5 -5 -2>, angularTarget=<0 0 0>}
      localJointOffset=<-5 1 -0.1>
     }]
    relativeSpaceBone {boneName=root}
    simSpaceSettings {maxAngularAcceleration=500, maxAngularVelocity=50}
   }
   n9 = Anim Dynamics - Bone: dyn_Cloth_volume_l_C01!(ComponentPose=@n8, bAlphaBoolEnabled:bool=True, GravityOverride="0,0,500", ExternalForce="0,0,0") -> (Pose) {
    alphaInputType=Bool
    angularBiasOverride=1
    angularDampingOverride=0.9999
    angularSpringConstant=5000
    bAngularSpring=true
    bOverrideAngularBias=true
    bOverrideAngularDamping=true
    bUseGravityOverride=true
    gravityScale=0.3
    simulationSpace=BoneRelative
    alphaBoolBlend {blendOption=Linear}
    boundBone {boneName=dyn_Cloth_volume_l_C01}
    componentAppliedLinearAccClamp=<3000 3000 3000>
    componentLinearAccScale=<0.5 0.5 0.5>
    componentLinearVelScale=<0.55 0.55 0.55>
    gravityOverride {z=500}
    physicsBodyDefinitions [{
      collisionType=InnerSphere
      sphereCollisionRadius=2
      boundBone {boneName=dyn_Cloth_volume_l_C01}
      boxExtents=<25 5 15>
      constraintSetup {angularLimitsMax=<5 5 15>, angularLimitsMin=<-5 -5 -2>, angularTarget=<0 0 0>}
      localJointOffset=<-5 1 -0.1>
     }]
    relativeSpaceBone {boneName=root}
    simSpaceSettings {maxAngularAcceleration=500, maxAngularVelocity=50}
   }
   n10 = Anim Dynamics - Bone: dyn_Cloth_volume_l_C02!(ComponentPose=@n9, bAlphaBoolEnabled:bool=True, GravityOverride:Vector="0,0,500", ExternalForce="0,0,0") -> (Pose) {
    alphaInputType=Bool
    angularBiasOverride=1
    angularDampingOverride=0.9999
    angularSpringConstant=5000
    bAngularSpring=true
    bOverrideAngularBias=true
    bOverrideAngularDamping=true
    bUseGravityOverride=true
    gravityScale=0.3
    simulationSpace=BoneRelative
    alphaBoolBlend {blendOption=Linear}
    boundBone {boneName=dyn_Cloth_volume_l_C02}
    componentAppliedLinearAccClamp=<3000 3000 3000>
    componentLinearAccScale=<0.5 0.5 0.5>
    componentLinearVelScale=<0.55 0.55 0.55>
    gravityOverride {z=500}
    physicsBodyDefinitions [{
      collisionType=InnerSphere
      sphereCollisionRadius=2
      boundBone {boneName=dyn_Cloth_volume_l_C02}
      boxExtents=<18 5 10>
      constraintSetup {angularLimitsMax=<5 5 15>, angularLimitsMin=<-5 -5 -2>, angularTarget=<0 0 0>}
      localJointOffset=<-5 1 -0.1>
     }]
    relativeSpaceBone {boneName=root}
    simSpaceSettings {maxAngularAcceleration=500, maxAngularVelocity=50}
   }
   n11 = Anim Dynamics - Bone: dyn_Cloth_volume_l_C01(ComponentPose=@n10, bAlphaBoolEnabled=True, GravityOverride="0,0,500", ExternalForce="0,0,0") {
    alphaInputType=Bool
    angularBiasOverride=1
    angularDampingOverride=0.9999
    angularSpringConstant=5000
    bAngularSpring=true
    bOverrideAngularBias=true
    bOverrideAngularDamping=true
    bUseGravityOverride=true
    gravityScale=0.3
    simulationSpace=BoneRelative
    alphaBoolBlend {blendOption=Linear}
    boundBone {boneName=dyn_Cloth_volume_l_C01}
    componentAppliedLinearAccClamp=<3000 3000 3000>
    componentLinearAccScale=<0.5 0.5 0.5>
    componentLinearVelScale=<0.55 0.55 0.55>
    gravityOverride {z=500}
    physicsBodyDefinitions [{
      collisionType=InnerSphere
      sphereCollisionRadius=2
      boundBone {boneName=dyn_Cloth_volume_l_C01}
      boxExtents=<25 5 15>
      constraintSetup {angularLimitsMax=<0 5 15>, angularLimitsMin=<0 -5 -2>, angularTarget=<0 0 0>}
      localJointOffset=<-5 1 -0.1>
     }]
    relativeSpaceBone {boneName=root}
    simSpaceSettings {maxAngularAcceleration=500, maxAngularVelocity=50}
   }
   n12 = Anim Dynamics - Bone: dyn_Cloth_volume_l_C02(ComponentPose=@n11, bAlphaBoolEnabled=True, GravityOverride="0,0,500", ExternalForce="0,0,0") {
    alphaInputType=Bool
    angularBiasOverride=1
    angularDampingOverride=0.9999
    angularSpringConstant=5000
    bAngularSpring=true
    bOverrideAngularBias=true
    bOverrideAngularDamping=true
    bUseGravityOverride=true
    gravityScale=0.3
    simulationSpace=BoneRelative
    alphaBoolBlend {blendOption=Linear}
    boundBone {boneName=dyn_Cloth_volume_l_C02}
    componentAppliedLinearAccClamp=<3000 3000 3000>
    componentLinearAccScale=<0.5 0.5 0.5>
    componentLinearVelScale=<0.55 0.55 0.55>
    gravityOverride {z=500}
    physicsBodyDefinitions [{
      collisionType=InnerSphere
      sphereCollisionRadius=2
      boundBone {boneName=dyn_Cloth_volume_l_C02}
      boxExtents=<18 5 10>
      constraintSetup {angularLimitsMax=<0 5 15>, angularLimitsMin=<0 -5 -2>, angularTarget=<0 0 0>}
      localJointOffset=<-5 1 -0.1>
     }]
    relativeSpaceBone {boneName=root}
    simSpaceSettings {maxAngularAcceleration=500, maxAngularVelocity=50}
   }
   n13 = Anim Dynamics - Bone: dyn_Cloth_volume_r_C01!(ComponentPose=@n12, Alpha=1, GravityOverride:Vector="0,0,-980", ExternalForce="0,0,0") -> (Pose) {
    angularBiasOverride=1
    angularDampingOverride=0.9999
    angularSpringConstant=5000
    bAngularSpring=true
    bOverrideAngularBias=true
    bOverrideAngularDamping=true
    bUseGravityOverride=true
    gravityScale=0.3
    simulationSpace=BoneRelative
    alphaBoolBlend {blendOption=Linear}
    boundBone {boneName=dyn_Cloth_volume_r_C01}
    componentAppliedLinearAccClamp=<3000 3000 3000>
    componentLinearAccScale=<0.5 0.5 0.5>
    componentLinearVelScale=<0.55 0.55 0.55>
    gravityOverride {z=-980}
    physicsBodyDefinitions [{
      collisionType=CustomSphere
      sphereCollisionRadius=2
      boundBone {boneName=dyn_Cloth_volume_r_C01}
      boxExtents=<45 45 45>
      constraintSetup {angularLimitsMax=<5 5 15>, angularLimitsMin=<-5 -5 -2>, angularTarget=<0 0 0>}
      localJointOffset=<-5 1 -0.1>
     }]
    relativeSpaceBone {boneName=root}
    simSpaceSettings {maxAngularAcceleration=500, maxAngularVelocity=50}
   }
   n14 = Anim Dynamics - Bone: dyn_Cloth_volume_r_C01!(ComponentPose=@n13, bAlphaBoolEnabled:bool=True, GravityOverride="0,0,-980", ExternalForce="0,0,0") -> (Pose) {
    alphaInputType=Bool
    angularBiasOverride=1
    angularDampingOverride=0.9999
    angularSpringConstant=5000
    bAngularSpring=true
    bOverrideAngularBias=true
    bOverrideAngularDamping=true
    bUseGravityOverride=true
    gravityScale=0.3
    simulationSpace=BoneRelative
    alphaBoolBlend {blendOption=Linear}
    boundBone {boneName=dyn_Cloth_volume_r_C01}
    componentAppliedLinearAccClamp=<3000 3000 3000>
    componentLinearAccScale=<0.5 0.5 0.5>
    componentLinearVelScale=<0.55 0.55 0.55>
    gravityOverride {z=-980}
    physicsBodyDefinitions [{
      collisionType=InnerSphere
      sphereCollisionRadius=2
      boundBone {boneName=dyn_Cloth_volume_r_C01}
      boxExtents=<25 5 15>
      constraintSetup {angularLimitsMax=<5 5 15>, angularLimitsMin=<-5 -5 -2>, angularTarget=<0 0 0>}
      localJointOffset=<-5 1 -0.1>
     }]
    relativeSpaceBone {boneName=root}
    simSpaceSettings {maxAngularAcceleration=500, maxAngularVelocity=50}
   }
   n15 = Anim Dynamics - Bone: dyn_Cloth_volume_r_C02!(ComponentPose=@n14, bAlphaBoolEnabled:bool=True, GravityOverride:Vector="0,0,-980", ExternalForce="0,0,0") -> (Pose) {
    alphaInputType=Bool
    angularBiasOverride=1
    angularDampingOverride=0.9999
    angularSpringConstant=5000
    bAngularSpring=true
    bOverrideAngularBias=true
    bOverrideAngularDamping=true
    bUseGravityOverride=true
    gravityScale=0.3
    simulationSpace=BoneRelative
    alphaBoolBlend {blendOption=Linear}
    boundBone {boneName=dyn_Cloth_volume_r_C02}
    componentAppliedLinearAccClamp=<3000 3000 3000>
    componentLinearAccScale=<0.5 0.5 0.5>
    componentLinearVelScale=<0.55 0.55 0.55>
    gravityOverride {z=-980}
    physicsBodyDefinitions [{
      collisionType=InnerSphere
      sphereCollisionRadius=2
      boundBone {boneName=dyn_Cloth_volume_r_C02}
      boxExtents=<18 5 10>
      constraintSetup {angularLimitsMax=<5 5 15>, angularLimitsMin=<-5 -5 -2>, angularTarget=<0 0 0>}
      localJointOffset=<-5 1 -0.1>
     }]
    relativeSpaceBone {boneName=root}
    simSpaceSettings {maxAngularAcceleration=500, maxAngularVelocity=50}
   }
   n16 = Anim Dynamics - Bone: dyn_Cloth_volume_r_C01(ComponentPose=@n15, bAlphaBoolEnabled=True, GravityOverride="0,0,-980", ExternalForce="0,0,0") {
    alphaInputType=Bool
    angularBiasOverride=1
    angularDampingOverride=0.9999
    angularSpringConstant=5000
    bAngularSpring=true
    bOverrideAngularBias=true
    bOverrideAngularDamping=true
    bUseGravityOverride=true
    gravityScale=0.3
    simulationSpace=BoneRelative
    alphaBoolBlend {blendOption=Linear}
    boundBone {boneName=dyn_Cloth_volume_r_C01}
    componentAppliedLinearAccClamp=<3000 3000 3000>
    componentLinearAccScale=<0.5 0.5 0.5>
    componentLinearVelScale=<0.55 0.55 0.55>
    gravityOverride {z=-980}
    physicsBodyDefinitions [{
      collisionType=InnerSphere
      sphereCollisionRadius=2
      boundBone {boneName=dyn_Cloth_volume_r_C01}
      boxExtents=<25 5 15>
      constraintSetup {angularLimitsMax=<0 5 15>, angularLimitsMin=<0 -5 -2>, angularTarget=<0 0 0>}
      localJointOffset=<-5 1 -0.1>
     }]
    relativeSpaceBone {boneName=root}
    simSpaceSettings {maxAngularAcceleration=500, maxAngularVelocity=50}
   }
   n17 = Anim Dynamics - Bone: dyn_Cloth_volume_r_C02(ComponentPose=@n16, bAlphaBoolEnabled=True, GravityOverride="0,0,-980", ExternalForce="0,0,0") {
    alphaInputType=Bool
    angularBiasOverride=1
    angularDampingOverride=0.9999
    angularSpringConstant=5000
    bAngularSpring=true
    bOverrideAngularBias=true
    bOverrideAngularDamping=true
    bUseGravityOverride=true
    gravityScale=0.3
    simulationSpace=BoneRelative
    alphaBoolBlend {blendOption=Linear}
    boundBone {boneName=dyn_Cloth_volume_r_C02}
    componentAppliedLinearAccClamp=<3000 3000 3000>
    componentLinearAccScale=<0.5 0.5 0.5>
    componentLinearVelScale=<0.55 0.55 0.55>
    gravityOverride {z=-980}
    physicsBodyDefinitions [{
      collisionType=InnerSphere
      sphereCollisionRadius=2
      boundBone {boneName=dyn_Cloth_volume_r_C02}
      boxExtents=<18 5 10>
      constraintSetup {angularLimitsMax=<0 5 15>, angularLimitsMin=<0 -5 -2>, angularTarget=<0 0 0>}
      localJointOffset=<-5 1 -0.1>
     }]
    relativeSpaceBone {boneName=root}
    simSpaceSettings {maxAngularAcceleration=500, maxAngularVelocity=50}
   }
   n18 = Anim Dynamics - Bone: dyn_wool_L_a!(ComponentPose=@n17, Alpha=1, GravityOverride:Vector="0,0,-980", ExternalForce="0,0,0") -> (Pose) {
    angularDampingOverride=0.9999
    angularSpringConstant=800
    bAngularSpring=true
    bOverrideAngularDamping=true
    bUseGravityOverride=true
    gravityScale=0.3
    simulationSpace=RootRelative
    alphaBoolBlend {blendOption=Linear}
    boundBone {boneName=dyn_wool_L_a}
    componentAppliedLinearAccClamp=<2000 2000 1600>
    componentLinearAccScale=<0.4 0.4 0.4>
    componentLinearVelScale=<0.45 0.45 0.55>
    gravityOverride {z=-980}
    physicsBodyDefinitions [{
      sphereCollisionRadius=2
      boundBone {boneName=dyn_wool_L_a}
      boxExtents=<8 10 15>
      constraintSetup {angularLimitsMax=<2 20 10>, angularLimitsMin=<-2 -20 -25>, angularTarget=<0 0 0>}
      localJointOffset=<8 0 0>
     }]
    relativeSpaceBone {boneName=root}
   }
   n19 = Anim Dynamics - Bone: dyn_wool_L_b!(ComponentPose=@n18, Alpha=1, GravityOverride:Vector="0,0,-980", ExternalForce="0,0,0") -> (Pose) {
    angularDampingOverride=1
    angularSpringConstant=800
    bAngularSpring=true
    bOverrideAngularDamping=true
    bUseGravityOverride=true
    gravityScale=0.3
    simulationSpace=RootRelative
    alphaBoolBlend {blendOption=Linear}
    boundBone {boneName=dyn_wool_L_b}
    componentAppliedLinearAccClamp=<2000 2000 1600>
    componentLinearAccScale=<0.4 0.4 0.4>
    componentLinearVelScale=<0.45 0.45 0.55>
    gravityOverride {z=-980}
    physicsBodyDefinitions [{
      sphereCollisionRadius=2
      boundBone {boneName=dyn_wool_L_b}
      boxExtents=<11 13 13>
      constraintSetup {twistAxis=AxisZ, angularLimitsMax=<2 20 10>, angularLimitsMin=<-2 -20 -25>, angularTarget=<0 0 0>}
      localJointOffset=<8 -2 0>
     }]
    relativeSpaceBone {boneName=root}
   }
   n20 = Anim Dynamics - Bone: dyn_wool_R_a!(ComponentPose=@n19, Alpha=1, GravityOverride:Vector="0,0,-980", ExternalForce="0,0,0") -> (Pose) {
    angularDampingOverride=0.9999
    angularSpringConstant=800
    bAngularSpring=true
    bOverrideAngularDamping=true
    bUseGravityOverride=true
    gravityScale=0.3
    simulationSpace=RootRelative
    alphaBoolBlend {blendOption=Linear}
    boundBone {boneName=dyn_wool_R_a}
    componentAppliedLinearAccClamp=<2000 2000 1600>
    componentLinearAccScale=<0.4 0.4 0.4>
    componentLinearVelScale=<0.45 0.45 0.55>
    gravityOverride {z=-980}
    physicsBodyDefinitions [{
      sphereCollisionRadius=2
      boundBone {boneName=dyn_wool_R_a}
      boxExtents=<8 10 15>
      constraintSetup {angularLimitsMax=<2 20 10>, angularLimitsMin=<-2 -20 -25>, angularTarget=<0 0 0>}
      localJointOffset=<8 0 0>
     }]
    relativeSpaceBone {boneName=root}
   }
   n21 = Anim Dynamics - Bone: dyn_wool_R_b!(ComponentPose=@n20, Alpha=1, GravityOverride:Vector="0,0,-980", ExternalForce="0,0,0") -> (Pose) {
    angularDampingOverride=1
    angularSpringConstant=800
    bAngularSpring=true
    bOverrideAngularDamping=true
    bUseGravityOverride=true
    gravityScale=0.3
    simulationSpace=RootRelative
    alphaBoolBlend {blendOption=Linear}
    boundBone {boneName=dyn_wool_R_b}
    componentAppliedLinearAccClamp=<2000 2000 1600>
    componentLinearAccScale=<0.4 0.4 0.4>
    componentLinearVelScale=<0.45 0.45 0.55>
    gravityOverride {z=-980}
    physicsBodyDefinitions [{
      sphereCollisionRadius=2
      boundBone {boneName=dyn_wool_R_b}
      boxExtents=<11 13 13>
      constraintSetup {twistAxis=AxisZ, angularLimitsMax=<2 20 10>, angularLimitsMin=<-2 -20 -25>, angularTarget=<0 0 0>}
      localJointOffset=<8 -2 0>
     }]
    relativeSpaceBone {boneName=root}
   }
   n22 = Anim Dynamics - Bone: dyn_wool_M_a!(ComponentPose=@n21, Alpha=1, GravityOverride:Vector="0,0,-980", ExternalForce="0,0,0") -> (Pose) {
    angularDampingOverride=0.999
    angularSpringConstant=800
    bAngularSpring=true
    bOverrideAngularDamping=true
    bUseGravityOverride=true
    gravityScale=0.3
    simulationSpace=RootRelative
    alphaBoolBlend {blendOption=Linear}
    boundBone {boneName=dyn_wool_M_a}
    componentAppliedLinearAccClamp=<2000 2000 1600>
    componentLinearAccScale=<0.4 0.4 0.4>
    componentLinearVelScale=<0.45 0.45 0.55>
    gravityOverride {z=-980}
    physicsBodyDefinitions [{
      sphereCollisionRadius=2
      boundBone {boneName=dyn_wool_M_a}
      boxExtents=<8 10 25>
      constraintSetup {angularLimitsMax=<2 5 5>, angularLimitsMin=<-2 -5 -50>, angularTarget=<0 0 0>}
      localJointOffset=<8 0 0>
     }]
    relativeSpaceBone {boneName=root}
   }
   n23 = Anim Dynamics - Bone: dyn_wool_M_b!(ComponentPose=@n22, Alpha=1, GravityOverride:Vector="0,0,-980", ExternalForce="0,0,0") -> (Pose) {
    angularDampingOverride=1
    angularSpringConstant=800
    bAngularSpring=true
    bOverrideAngularDamping=true
    bUseGravityOverride=true
    gravityScale=0.3
    simulationSpace=RootRelative
    alphaBoolBlend {blendOption=Linear}
    boundBone {boneName=dyn_wool_M_b}
    componentAppliedLinearAccClamp=<2000 2000 1600>
    componentLinearAccScale=<0.4 0.4 0.4>
    componentLinearVelScale=<0.45 0.45 0.55>
    gravityOverride {z=-980}
    physicsBodyDefinitions [{
      sphereCollisionRadius=2
      boundBone {boneName=dyn_wool_M_b}
      boxExtents=<8 10 20>
      constraintSetup {angularLimitsMax=<2 5 5>, angularLimitsMin=<-2 -5 -50>, angularTarget=<0 0 0>}
      localJointOffset=<8 0 0>
     }]
    relativeSpaceBone {boneName=root}
   }
   n24 = Anim Dynamics - Bone: dyn_wool_M_f!(ComponentPose=@n23, Alpha=1, GravityOverride:Vector="0,0,-980", ExternalForce="0,0,0") -> (Pose) {
    angularDampingOverride=1
    angularSpringConstant=800
    bAngularSpring=true
    bOverrideAngularDamping=true
    bUseGravityOverride=true
    gravityScale=0.3
    simulationSpace=RootRelative
    alphaBoolBlend {blendOption=Linear}
    boundBone {boneName=dyn_wool_M_f}
    componentAppliedLinearAccClamp=<2000 2000 1600>
    componentLinearAccScale=<0.4 0.4 0.4>
    componentLinearVelScale=<0.45 0.45 0.55>
    gravityOverride {z=-980}
    physicsBodyDefinitions [{
      sphereCollisionRadius=2
      boundBone {boneName=dyn_wool_M_f}
      boxExtents=<8 10 20>
      constraintSetup {angularLimitsMax=<2 5 5>, angularLimitsMin=<-2 -5 -10>, angularTarget=<0 0 0>}
      localJointOffset=<8 0 0>
     }]
    relativeSpaceBone {boneName=root}
   }
   n25 = Anim Dynamics - Bone: dyn_chain_M_FL_01!(ComponentPose=@n24, Alpha=1, GravityOverride:Vector="0,0,-980", ExternalForce="0,0,0") -> (Pose) {
    angularDampingOverride=0.9999
    angularSpringConstant=5000
    bAngularSpring=true
    bOverrideAngularDamping=true
    bUseGravityOverride=true
    gravityScale=0.3
    simulationSpace=RootRelative
    alphaBoolBlend {blendOption=Linear}
    boundBone {boneName=dyn_chain_M_FL_01}
    componentAppliedLinearAccClamp=<2000 2000 1600>
    componentLinearAccScale=<0.4 0.4 0.4>
    componentLinearVelScale=<0.45 0.45 0.55>
    gravityOverride {z=-980}
    physicsBodyDefinitions [{
      sphereCollisionRadius=2
      boundBone {boneName=dyn_chain_M_FL_01}
      boxExtents=<8 10 20>
      constraintSetup {angularLimitsMax=<2 5 35>, angularLimitsMin=<-2 -5 -10>, angularTarget=<0 0 0>}
      localJointOffset=<3.5 0 0>
     }]
    relativeSpaceBone {boneName=root}
   }
   n26 = Anim Dynamics - Bone: dyn_chain_M_FL_02!(ComponentPose=@n25, Alpha=1, GravityOverride:Vector="0,0,-980", ExternalForce="0,0,0") -> (Pose) {
    angularDampingOverride=0.9999
    angularSpringConstant=5000
    bAngularSpring=true
    bOverrideAngularDamping=true
    bUseGravityOverride=true
    gravityScale=0.3
    simulationSpace=RootRelative
    alphaBoolBlend {blendOption=Linear}
    boundBone {boneName=dyn_chain_M_FL_02}
    componentAppliedLinearAccClamp=<2000 2000 1600>
    componentLinearAccScale=<0.4 0.4 0.4>
    componentLinearVelScale=<0.45 0.45 0.55>
    gravityOverride {z=-980}
    physicsBodyDefinitions [{
      sphereCollisionRadius=2
      boundBone {boneName=dyn_chain_M_FL_02}
      boxExtents=<11 13 13>
      constraintSetup {twistAxis=AxisZ, angularLimitsMax=<2 20 35>, angularLimitsMin=<-2 -20 -10>, angularTarget=<0 0 0>}
      localJointOffset=<6 -2 0>
     }]
    relativeSpaceBone {boneName=root}
   }
   n27 = Anim Dynamics - Bone: dyn_chain_M_FR_01!(ComponentPose=@n26, Alpha=1, GravityOverride:Vector="0,0,-980", ExternalForce="0,0,0") -> (Pose) {
    angularDampingOverride=0.9999
    angularSpringConstant=5000
    bAngularSpring=true
    bOverrideAngularDamping=true
    bUseGravityOverride=true
    gravityScale=0.3
    simulationSpace=RootRelative
    alphaBoolBlend {blendOption=Linear}
    boundBone {boneName=dyn_chain_M_FR_01}
    componentAppliedLinearAccClamp=<2000 2000 1600>
    componentLinearAccScale=<0.4 0.4 0.4>
    componentLinearVelScale=<0.45 0.45 0.55>
    gravityOverride {z=-980}
    physicsBodyDefinitions [{
      sphereCollisionRadius=2
      boundBone {boneName=dyn_chain_M_FR_01}
      boxExtents=<8 10 20>
      constraintSetup {angularLimitsMax=<2 5 35>, angularLimitsMin=<-2 -5 -10>, angularTarget=<0 0 0>}
      localJointOffset=<3.5 0 0>
     }]
    relativeSpaceBone {boneName=root}
   }
   n28 = Anim Dynamics - Bone: dyn_chain_M_FR_02!(ComponentPose=@n27, Alpha=1, GravityOverride:Vector="0,0,-980", ExternalForce="0,0,0") -> (Pose) {
    angularDampingOverride=0.9999
    angularSpringConstant=5000
    bAngularSpring=true
    bOverrideAngularDamping=true
    bUseGravityOverride=true
    gravityScale=0.3
    simulationSpace=RootRelative
    alphaBoolBlend {blendOption=Linear}
    boundBone {boneName=dyn_chain_M_FR_02}
    componentAppliedLinearAccClamp=<2000 2000 1600>
    componentLinearAccScale=<0.4 0.4 0.4>
    componentLinearVelScale=<0.45 0.45 0.55>
    gravityOverride {z=-980}
    physicsBodyDefinitions [{
      sphereCollisionRadius=2
      boundBone {boneName=dyn_chain_M_FR_02}
      boxExtents=<11 13 13>
      constraintSetup {twistAxis=AxisZ, angularLimitsMax=<2 20 35>, angularLimitsMin=<-2 -20 -10>, angularTarget=<0 0 0>}
      localJointOffset=<6 -2 0>
     }]
    relativeSpaceBone {boneName=root}
   }
   Output Pose(Result=Component To Local(ComponentPose=@n28)) {layerGroup=None}
  ]}
 EventGraph {nodes [
   Event BlueprintUpdateAnimation()
   > n1 = "Gravity Override State Function (Parameter Asset) (Nativized)"(
    CharacterPartAnimInstance=Self-Reference()
    GameTimeInSeconds=GetGameTimeInSeconds()
    Parameters=/BRCosmetics/Characters/Player/Female/Medium/Bodies/F_MED_IndoorLaceScore/Meshes/F_MED_IndoorLaceScore_Parameters
   )
   > Set GravityOverride(GravityOverride=@n1)
   > Set RBAN Settings Out(RBAN Settings Out=@n1.OutRigidBodyAnimNodeParameters)
   > Set Not bIsSkyDiving(Not bIsSkyDiving:bool=NOT Boolean(A=Get bIsSkyDiving())) -> (Output_Get:bool)
  ]}
]
variables [
 clothing {type=boolean}
 GravityOverride {type=Vector}
 Not bIsSkyDiving {type=boolean}
 RBAN Settings Out {type=FortRigidBodyAnimNodeParameters}
]
